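# Control Rig for the ChewyStaveCurve backpack cosmetic that adjusts the pack's bones during skydiving.
# The skydive float and diving variables are each smoothed by an AlphaInterp and used as the weight of a Modify Transforms step.
# The float weight drives transforms on the dangling chain bones, including rotations.
# The diving weight shifts and shrinks the whole pack through base_pack.
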ControlRigBlueprint /BRCosmetics/Accessories/FORT_Backpacks/Backpack_M_MED_ChewyStaveCurve/Meshes/M_MED_ChewyStaveCurve_Pack_CtrlRig
graphs [
 RigVMFunctionLibrary {links [], nodes []}
 RigVMModel {nodes [
   RigUnit_BeginExecution = &RigUnit_BeginExecution()
   n1 = &RigVMFunction_AlphaInterp(InterpSpeedDecreasing=5, InterpSpeedIncreasing=5, Scale=1, bClampResult=False, Value=RigVMVariableNode() {variable=SkydiveFloat_Float}.Value)
   &RigUnit_ModifyTransforms(
    ItemToModify="((Item=(Type=Bone,Name=\"dyn_chain_01\"),Transform=(Rotation=(X=0.0151,Y=-0.0858,Z=0.173,W=0.9811),Translation=(X=0,Y=0,Z=0),Scale3D=(X=1,Y=1,Z=1))),(Item=(Type=Bone,Name=\"dyn_chain_02\"),Transform=(Rotation=(X=0.0151,Y=-0.0858,Z=0.173,W=0.9811),Translation=( ...(truncated)"
    WeightMinimum=0
    Weight=@n1.Result
    ExecutePin=@RigUnit_BeginExecution.ExecutePin
   )
   AlphaInterp = &RigVMFunction_AlphaInterp(
    InterpSpeedDecreasing=5
    InterpSpeedIncreasing=5
    Scale=1
    bClampResult=False
    Value=RigVMVariableNode() {variable=SkyDive_Diving_Float}.Value
   )
   > &RigUnit_ModifyTransforms(
    ItemToModify=((Item=(Type=Bone,Name="base_pack"),Transform=(Rotation=(X=0,Y=0,Z=0,W=1),Translation=(X=0,Y=-5,Z=0),Scale3D=(X=0.8,Y=0.8,Z=0.8))))
    WeightMinimum=0
    Weight=@AlphaInterp.Result
   )
  ]}
]
hierarchy [
 root {loc=<0 0 0>, rot=<0 0 0 1>}
 pack {parent=root, loc=<0 -0.7316 0>, rot=<0 0 0 1>}
 base_pack {parent=pack, loc=<0 0 0>, rot=<0 0 0 1>}
 base_chain_01 {parent=base_pack, loc=<-11.4422 -5.5432 3.7971>, rot=<-0.2926 0.0031 0.9562 0.007>}
 dyn_chain_01 {parent=base_chain_01, loc=<0 0 0>, rot=<0 0 0 1>}
 dyn_chain_02 {parent=dyn_chain_01, loc=<1.9449 0 0>, rot=<0.0086 -0.2805 0.0168 -0.9597>}
 dyn_chain_03 {parent=dyn_chain_02, loc=<1.5397 0 0>, rot=<-0.0038 0.0213 -0.021 -0.9995>}
 dyn_chain_04 {parent=dyn_chain_03, loc=<3.0617 0 0>, rot=<0 0 0 1>}
 base_handle_01 {parent=base_pack, loc=<0.0554 -5.5494 4.6999>, rot=<-0.4701 0.5182 -0.4868 -0.523>}
 dyn_handle_01 {parent=base_handle_01, loc=<0 0 0>, rot=<0 0 0 1>}
 dyn_handle_02 {parent=dyn_handle_01, loc=<3.8429 0 0>, rot=<0 -0.0075 0 -1>}
 dyn_handle_03 {parent=dyn_handle_02, loc=<3.3182 0 0>, rot=<0 0 0 1>}
]
variables [
 SkyDive_Diving_Float {type=double, default=0}
 SkydiveFloat_Float {type=double, default=0}
]
