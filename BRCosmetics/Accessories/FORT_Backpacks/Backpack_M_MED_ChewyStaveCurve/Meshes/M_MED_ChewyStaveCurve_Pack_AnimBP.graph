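# Animation Blueprint for the ChewyStaveCurve backpack cosmetic, adding secondary motion and skydive-aware posing.
# Five sequential Anim Dynamics nodes simulate three chain and two handle bones with root-relative angular springs, using downward gravity for the chain and upward for the handles.
# The result feeds the pack's Control Rig.
# Each update converts the skydive floating and dive-mode booleans to floats that the rig receives as inputs.
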AnimBlueprint /BRCosmetics/Accessories/FORT_Backpacks/Backpack_M_MED_ChewyStaveCurve/Meshes/M_MED_ChewyStaveCurve_Pack_AnimBP
target_skeleton=&/M_MED_ChewyStaveCurve_Pack_Skeleton
graphs [
 AnimGraph {nodes [
   n1 = Anim Dynamics - Bone: dyn_chain_01(ComponentPose=Local To Component(LocalPose=Input Pose()), Alpha=1, ExternalForce="0,0,0") {
    angularDampingOverride=1
    angularSpringConstant=5000
    bAngularSpring=true
    bOverrideAngularDamping=true
    bUseGravityOverride=true
    gravityScale=0.35
    simulationSpace=RootRelative
    boundBone {boneName=dyn_chain_01}
    componentAppliedLinearAccClamp=<50000 50000 50000>
    componentLinearAccScale=<0.8 0.8 0.8>
    componentLinearVelScale=<0.8 0.8 0.8>
    gravityOverride {z=-500}
    physicsBodyDefinitions [{
      boundBone {boneName=dyn_chain_01}
      boxExtents=<15 20 20>
      constraintSetup {angularLimitsMax=<0 -2 10>, angularLimitsMin=<0 -50 -10>, angularTarget=<1 0 0>}
      localJointOffset=<2 0 0>
     }]
   }
   n2 = Anim Dynamics - Bone: dyn_chain_02(ComponentPose=@n1, Alpha=1, ExternalForce="0,0,0") {
    angularDampingOverride=0.9999
    angularSpringConstant=2500
    bAngularSpring=true
    bOverrideAngularDamping=true
    bUseGravityOverride=true
    gravityScale=0.35
    simulationSpace=RootRelative
    boundBone {boneName=dyn_chain_02}
    componentAppliedLinearAccClamp=<50000 50000 50000>
    componentLinearAccScale=<0.8 0.8 0.8>
    componentLinearVelScale=<0.8 0.8 0.8>
    gravityOverride {z=-500}
    physicsBodyDefinitions [{
      boundBone {boneName=dyn_chain_02}
      boxExtents=<10 15 15>
      constraintSetup {angularLimitsMax=<0 -2 30>, angularLimitsMin=<0 -70 -30>, angularTarget=<1 0 0>}
      localJointOffset=<2.5 0 0>
     }]
   }
   n3 = Anim Dynamics - Bone: dyn_chain_03(ComponentPose=@n2, Alpha=1, ExternalForce="0,0,0") {
    angularDampingOverride=1
    angularSpringConstant=4500
    bAngularSpring=true
    bOverrideAngularDamping=true
    bUseGravityOverride=true
    gravityScale=0.35
    simulationSpace=RootRelative
    boundBone {boneName=dyn_chain_03}
    componentAppliedLinearAccClamp=<50000 50000 50000>
    componentLinearAccScale=<0.8 0.8 0.8>
    componentLinearVelScale=<0.8 0.8 0.8>
    gravityOverride {z=-980}
    physicsBodyDefinitions [{
      boundBone {boneName=dyn_chain_03}
      boxExtents=<15 20 20>
      constraintSetup {angularLimitsMax=<0 -2 20>, angularLimitsMin=<0 -80 -20>, angularTarget=<1 0 0>}
      localJointOffset=<3 0 0>
     }]
   }
   n4 = Anim Dynamics - Bone: dyn_handle_01(ComponentPose=@n3, Alpha=1, ExternalForce="0,0,0") {
    angularDampingOverride=1
    angularSpringConstant=5000
    bAngularSpring=true
    bOverrideAngularDamping=true
    bUseGravityOverride=true
    gravityScale=0.35
    simulationSpace=RootRelative
    boundBone {boneName=dyn_handle_01}
    componentAppliedLinearAccClamp=<20000 20000 20000>
    componentLinearAccScale=<0.5 0.5 0.5>
    componentLinearVelScale=<0.5 0.5 0.5>
    gravityOverride {z=980}
    physicsBodyDefinitions [{
      boundBone {boneName=dyn_handle_01}
      boxExtents=<30 40 40>
      constraintSetup {angularLimitsMax=<0 25 15>, angularLimitsMin=<0 -25 -15>, angularTarget=<1 0 0>}
      localJointOffset=<3 0 0>
     }]
    simSpaceSettings {maxAngularAcceleration=500, maxAngularVelocity=500, simSpaceAngularAlpha=0.2}
   }
   n5 = Anim Dynamics - Bone: dyn_handle_02(ComponentPose=@n4, Alpha=1, ExternalForce="0,0,0") {
    angularDampingOverride=1
    angularSpringConstant=5000
    bAngularSpring=true
    bOverrideAngularDamping=true
    bUseGravityOverride=true
    gravityScale=0.35
    simulationSpace=RootRelative
    boundBone {boneName=dyn_handle_02}
    componentAppliedLinearAccClamp=<20000 20000 20000>
    componentLinearAccScale=<0.5 0.5 0.5>
    componentLinearVelScale=<0.5 0.5 0.5>
    gravityOverride {z=980}
    physicsBodyDefinitions [{
      boundBone {boneName=dyn_handle_02}
      boxExtents=<30 40 40>
      constraintSetup {angularLimitsMax=<0 25 15>, angularLimitsMin=<0 -25 -15>, angularTarget=<1 0 0>}
      localJointOffset=<3 0 0>
     }]
    simSpaceSettings {maxAngularAcceleration=500, maxAngularVelocity=500, simSpaceAngularAlpha=0.2}
   }
   n6 = Control Rig!(
    Source=Component To Local(ComponentPose=@n5)
    Alpha
    SkydiveFloat_Float:double=Get SkyDiveFloating_Float() -> (SkyDiveFloating_Float:double)
    SkyDive_Diving_Float:double=Get SkyDive_Diving_Float() -> (SkyDive_Diving_Float:double)
   ) -> (Pose) {controlRigAssetReference {blueprintRigClass=&/M_MED_ChewyStaveCurve_Pack_CtrlRig.M_MED_ChewyStaveCurve_Pack_CtrlRig_C}}
   Output Pose(Result=@n6) {layerGroup=None}
  ]}
 EventGraph {nodes [
   Event BlueprintUpdateAnimation()
   > Set SkyDiveFloating_Float(SkyDiveFloating_Float:double="To Float (Boolean)"(InBool=Get bIsSkydiveFloating())) -> (Output_Get:double)
   > Set SkyDive_Diving_Float(SkyDive_Diving_Float:double="To Float (Boolean)"(InBool=Get bIsSkydiveDiveMode())) -> (Output_Get:double)
   TryGetPawnOwner()
  ]}
]
variables [
 SkyDive_Diving_Float {type=number}
 SkyDiveFloating_Float {type=number}
]
