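# Physics asset for the ChewyStaveCurve backpack cosmetic mesh, defining collision bodies and joints for a hanging chain on the pack.
# A single capsule covers the base pack bone, and three small capsules cover the dyn_chain bones.
# Constraints link the chain bones in sequence from the pack, with limited swing on both axes and twist limits set to zero degrees.
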PhysicsAsset /BRCosmetics/Accessories/FORT_Backpacks/Backpack_M_MED_ChewyStaveCurve/Meshes/M_MED_ChewyStaveCurve_Pack_Physics
bodies {
 base_pack [{shapeType=Capsule, length=8.0541, radius=6.5388, center=<-7.4366 -5.5465 -2.4666>, rotation=<rot 0 0 -90>}]
 dyn_chain_01 [{
   shapeType=Capsule
   length=0.7524
   radius=1.0631
   shapeName=dyn_chain_01_capsule
   center=<0.9614 0.0117 0.1488>
   rotation=<rot -0.0006 90.6999 -94.1448>
  }]
 dyn_chain_02 [{
   shapeType=Capsule
   length=0.8636
   radius=1.0068
   shapeName=dyn_chain_02_capsule
   center=<0.7087 -0.0059 0.1018>
   rotation=<rot 0.3147 91.6484 -81.4512>
  }]
 dyn_chain_03 [{
   shapeType=Capsule
   length=1.857
   radius=1.0901
   shapeName=dyn_chain_03_capsule
   center=<1.4149 -0.0327 0.1975>
   rotation=<rot 0 89.8944 -89.8841>
  }]
}
constraints [
 {bone1Name=dyn_chain_01, bone2Name=base_pack, swing1LimitDegrees=15, swing2LimitDegrees=45, twistLimitDegrees=0}
 {bone1Name=dyn_chain_02, bone2Name=dyn_chain_01, swing1LimitDegrees=15, swing2LimitDegrees=50, twistLimitDegrees=0}
 {bone1Name=dyn_chain_03, bone2Name=dyn_chain_02, swing1LimitDegrees=15, swing2LimitDegrees=50, twistLimitDegrees=0}
]
